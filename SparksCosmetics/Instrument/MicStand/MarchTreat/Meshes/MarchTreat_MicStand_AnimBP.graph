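# Animation Blueprint for the MarchTreat mic stand that fakes an audio-reactive material loop in frontend and preview displays, with no pose logic.
# On initialize it sets IsPreview from display-context, owner and component checks, then cycles BeginLoop on and off using Delays.
# Each update in preview eases a value toward that state, maps it through a ramp curve and writes it to the EnableFrontendFakeAudioLoop material parameter.
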AnimBlueprint /SparksCosmetics/Instrument/MicStand/MarchTreat/Meshes/MarchTreat_MicStand_AnimBP
target_skeleton=/SparksCharacterCommon/Instrument/MicStand/Generic/Meshes/Generic_MicStand_Skeleton
graphs [
 AnimGraph {nodes [Output Pose(Result=Input Pose())]}
 EventGraph {nodes [
   n1 = OnFakeLoopTimelineUpdate!() -> (OutputDelegate, then, TransitionValue:float=~) {is=title K2Node_CustomEvent}
   > SetScalarParameterValueOnMaterials!(
    execute
    self:Mesh Component Object Reference=GetOwningComponent()
    ParameterName:Name=EnableFrontendFakeAudioLoop
    ParameterValue:float=@n1.TransitionValue
   ) -> (then)
   n2 = Event BlueprintUpdateAnimation()
   > Transition!(execute, self, DeltaTimeX:float=@n2.DeltaTimeX) -> (then)
   Event BlueprintInitializeAnimation()
   n3 = Contains(SearchIn=GetDisplayName(Object=GetOwningComponent()), Substring=SkeletalMeshComponentBudgeted)
   n4 = IsInstrumentProp!(Instrument ID:int=4000) -> (OutputPin:bool)
   n5 = Cast To BP_SparksCosmeticPreviewDisplay_Base!(Object=GetOwningActor()) -> (AsBP Sparks Cosmetic Preview Display Base:BP Sparks Cosmetic Preview Display Base Object Reference, bSuccess:bool=~) {is=title K2Node_DynamicCast}
   n6 = IsCosmeticObjectDisplayedInFrontendContext!(CosmeticObject:Object Reference) -> (ReturnValue:bool)
   n7 = IsCosmeticObjectDisplayedAsPreview!(CosmeticObject:Object Reference) -> (ReturnValue:bool)
   > n8 = Set IsPreview(IsPreview:bool=AND Boolean(A=OR Boolean(A=@n6, B=@n7), B=OR Boolean!(A=@n5.bSuccess, B=@n3, C:bool=@n4) -> (ReturnValue))) -> (Output_Get:bool)
   > n9 = Branch(Condition=@n8)
   Set BeginLoop(execute=[@n10.then, @n9.then], BeginLoop:bool=true) -> (Output_Get:bool)
   > Delay(Duration=float + float(A=Get FakeAmplitudeOnLength() -> (FakeAmplitudeOnLength:float), B=1))
   > Set BeginLoop(BeginLoop=false)
   > Delay(Duration=0.8)
   > n10 = Delay(Duration=Get FakeAmplitudeOffLength() -> (FakeAmplitudeOffLength:float))
   K2Node_Event_2 = Transition!() -> (OutputDelegate, then, DeltaTimeX:float=~) {is=title K2Node_CustomEvent}
   > Branch(Condition=Get IsPreview() -> (IsPreview:bool))
   n11 = Get BeginLoop() -> (BeginLoop:bool)
   > n12 = Set Current(Current:double=FInterpTo_Constant!(
     Current:double=Get Current() -> (Current:double)
     Target:double="To Float (Boolean)"(InBool=@n11)
     DeltaTime:double=@K2Node_Event_2.DeltaTimeX
     InterpSpeed:double=1
    ) -> (ReturnValue:double)) -> (Output_Get:double)
   > OnFakeLoopTimelineUpdate!(
    execute
    self
    TransitionValue:float=GetFloatValueFromCurve!(InCurve:Curve Float Object Reference=Get Curve() -> (Curve:Curve Float Object Reference), Time:float=@n12) -> (OutValue:float, ReturnValue:bool)
   ) -> (then) {is=title K2Node_CallFunction}
   SetFloatParameter!(execute, self:FXSystem Component Object Reference, ParameterName:Name=EnableFrontendFakeAudioLoop, Param:float=@n1.TransitionValue) -> (then)
  ]}
 IsInstrumentProp {nodes [
   n1 = IsInstrumentProp!() -> (then, Instrument ID:int=1002) {is=title K2Node_FunctionEntry}
   > n2 = Cast To PlayerPawn_Athena(Object=GetOwningActor())
   n3 = Get EmoteMeshCompDatas!(self:Fort Pawn Object Reference=@n2) -> (EmoteMeshCompDatas:Map of Integers to Emote Mesh Comp Data Structures)
   > Return Node!(
    execute
    OutputPin:bool="Equal (Object)"(
     A=Find!(TargetMap:Map of Integers to Emote Mesh Comp Data Structures=@n3, Key:"Integer (by ref)"=@n1) -> (
      Value_MeshComp:Mesh Component Object Reference
      Value_PropExclusionTag:Gameplay Tag Structure
      Value_bTickEvenWhenNotRendered:bool=~
      ReturnValue
     )
     B=GetOwningComponent()
    )
   )
   Return Node(execute=@n2.CastFailed, OutputPin=false)
  ]}
]
variables [
 BeginLoop {type=boolean}
 Current {type=number}
 Curve {type=CurveFloat, default=/SparksCosmetics/Instrument/General/Anim/CRV_SparksCosmetics_SmoothRampUp}
 FakeAmplitudeOffLength {type=number, default=4}
 FakeAmplitudeOnLength {type=number, default=7}
 IsPreview {type=boolean}
]
